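; BP_Electrical_DamagedCable is an environmental effect actor for a damaged electrical cable that pairs sparking particles with zap audio.
; On BeginPlay it binds to the P_Electrical_DamagedCable particle component's OnParticleSpawn event.
; When the spawn event name is "Parent_Spawn", it plays the Snow_DamagedCable_Zap cue at the AmbientAudio component's location, so each zap follows the particle emission.
; The tick, overlap and construction script graphs are empty.
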(event EventBeginPlay
  ;[1 1 1] Add the sound to Branch = True
  (EventDispatchers|BindEventtoOnParticleSpawn (Variables|Default|GetP_Electrical_DamagedCable) (delegate Custom|Parent_Spawn)))

(event Custom|Parent_Spawn
  (params (EventName :type Name) (EmitterTime :type Float%20%28single-precision%29) (Location :type Vector) (Velocity :type Vector))
  ;[1 1 1] Add the sound to Branch = True
  (if (Utilities|Name|Equal(Name) EventName "Parent_Spawn")
    (Audio|PlaySoundatLocation "/Figment_S09/Athena/Sounds/Events/Dopey/Snow/Snow_DamagedCable_Zap_Cue.Snow_DamagedCable_Zap_Cue" (Transformation|GetWorldLocation (Variables|Default|GetAmbientAudio)))))

(event Collision|EventActorBeginOverlap
  (params (OtherActor :type Actor%20Object%20Reference)))

(event EventTick
  (params (DeltaSeconds :type Float%20%28single-precision%29)))

(fn ConstructionScript)
